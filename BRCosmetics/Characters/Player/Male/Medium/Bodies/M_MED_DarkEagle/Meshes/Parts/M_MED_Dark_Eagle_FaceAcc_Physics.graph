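# Physics asset for the face accessory mesh of the medium male Dark Eagle cosmetic body, defining physics for its hood bones.
# Thin box bodies cover the left and right hood bones, a smaller box covers the hood tip, and a capsule represents the head.
# Each hood body is constrained directly to the head with limited swing and tighter twist, with the tip allowed wider swing than the sides.
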PhysicsAsset /BRCosmetics/Characters/Player/Male/Medium/Bodies/M_MED_DarkEagle/Meshes/Parts/M_MED_Dark_Eagle_FaceAcc_Physics
bodies {
 dyn_hood_l [{shapeType=Box, extentX=5.8676, extentY=6.0765, extentZ=1, center=<1.9899 1.4835 -0.0001>, rotation=<rot 0 -13.9539 0>}]
 dyn_hood_r [{shapeType=Box, extentX=5.0745, extentY=5.3833, extentZ=1.4681, center=<2.1388 1.3786 -0.3928>, rotation=<rot 0 -13.9539 0>}]
 dyn_hood_tip [{shapeType=Box, extentX=1.4198, extentY=3.1205, extentZ=2.8647}]
 head [{shapeType=Capsule, length=11, radius=5.5, rotation=<rot 90 0 0>}]
}
constraints [
 {bone1Name=dyn_hood_r, bone2Name=head, swing1LimitDegrees=20, swing2LimitDegrees=20, twistLimitDegrees=8}
 {bone1Name=dyn_hood_l, bone2Name=head, swing1LimitDegrees=20, swing2LimitDegrees=20, twistLimitDegrees=8}
 {bone1Name=dyn_hood_tip, bone2Name=head, swing1LimitDegrees=30, swing2LimitDegrees=30, twistLimitDegrees=6}
]
